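# Physics asset for the dangling strings on the Kpop fashion hat's face accessory.
# Left and right chains of three thin capsules hang from a small master string body, each joint constrained with a tight twist limit.
# The master body links to spine_04, which links to spine_03, both through free-rotation constraints.
# The two spine bones carry large capsules that approximate the upper torso.
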PhysicsAsset /BRCosmetics/Accessories/Hats/M_MED_Kpop_Fashion_Hat/Meshes/M_MED_Kpop_Fashion_FaceAcc_Dyn
bodies {
 dyn_string_MSTR [{shapeType=Capsule, length=0.505, radius=0.505, rotation=<rot 0 -90 0>}]
 dyn_string_l_01 [{shapeType=Capsule, length=6, radius=1, center=<4 0 0>, rotation=<rot 90 0 0>}]
 dyn_string_l_02 [{shapeType=Capsule, length=6, radius=1, center=<4 0 0>, rotation=<rot 90 0 0>}]
 dyn_string_l_03 [{shapeType=Capsule, length=6, radius=1, center=<4 0 0>, rotation=<rot 90 0 0>}]
 dyn_string_r_01 [{shapeType=Capsule, length=6, radius=1, center=<4 0 0>, rotation=<rot 90 0 0>}]
 dyn_string_r_02 [{shapeType=Capsule, length=6, radius=1, center=<4 0 0>, rotation=<rot 90 0 0>}]
 dyn_string_r_03 [{shapeType=Capsule, length=6, radius=1, center=<4 0 0>, rotation=<rot 90 0 0>}]
 spine_03 [{shapeType=Capsule, length=17.1575, radius=17.032, center=<4.9458 -4.8495 0>}]
 spine_04 [{shapeType=Capsule, length=16.7583, radius=17.4311, center=<5.5087 -6.0009 0>}]
}
constraints [
 {bone1Name=dyn_string_r_02, bone2Name=dyn_string_r_01, swing1LimitDegrees=45, swing2LimitDegrees=45, twistLimitDegrees=5}
 {bone1Name=dyn_string_r_03, bone2Name=dyn_string_r_02, swing1LimitDegrees=45, swing2LimitDegrees=45, twistLimitDegrees=5}
 {bone1Name=dyn_string_l_02, bone2Name=dyn_string_l_01, swing1LimitDegrees=45, swing2LimitDegrees=45, twistLimitDegrees=5}
 {bone1Name=dyn_string_l_03, bone2Name=dyn_string_l_02, swing1LimitDegrees=45, swing2LimitDegrees=45, twistLimitDegrees=5}
 {bone1Name=dyn_string_r_01, bone2Name=dyn_string_MSTR, swing1LimitDegrees=45, swing2LimitDegrees=45, twistLimitDegrees=5}
 {bone1Name=dyn_string_l_01, bone2Name=dyn_string_MSTR, swing1LimitDegrees=45, swing2LimitDegrees=45, twistLimitDegrees=5}
 {
  bone1Name=spine_04
  bone2Name=spine_03
  swing1LimitDegrees=45
  swing1Motion=Free
  swing2LimitDegrees=45
  swing2Motion=Free
  twistLimitDegrees=45
  twistMotion=Free
 }
 {
  bone1Name=dyn_string_MSTR
  bone2Name=spine_04
  swing1LimitDegrees=45
  swing1Motion=Free
  swing2LimitDegrees=45
  swing2Motion=Free
  twistLimitDegrees=45
  twistMotion=Free
 }
]
